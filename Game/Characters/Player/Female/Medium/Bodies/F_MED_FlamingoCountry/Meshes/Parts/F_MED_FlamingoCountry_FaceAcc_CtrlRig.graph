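# Control Rig for FlamingoCountry female face accessories, correcting hat loop, string and straw bones as the character moves.
# The hat loop follows neck_02, with spherical pose readers on head rotation relative to neck_01 weighting corrective offsets.
# The straw is constrained between jaw and lower-left lip, then nudged.
# The Melee variable scales another head-driven hat offset and weights constraints pinning loop translation and both strings mostly or fully to spine_05.
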ControlRigBlueprint /Game/Characters/Player/Female/Medium/Bodies/F_MED_FlamingoCountry/Meshes/Parts/F_MED_FlamingoCountry_FaceAcc_CtrlRig
graphs [
 RigVMFunctionLibrary {links [], nodes []}
 RigVMModel {nodes [
   RigUnit_BeginExecution = &RigUnit_BeginExecution()
   &RigUnit_ParentConstraint(
    AdvancedSettings=(InterpolationType=Average,RotationOrderForFilter=XZY)
    Child=(Type=Bone,Name="dyn_hat_loop_02")
    Parents=((Item=(Type=Bone,Name="neck_02"),Weight=1))
    Weight=1
    ExecutePin=@RigUnit_BeginExecution.ExecutePin
   )
   > n1 = &RigUnit_SphericalPoseReader(
    ActiveRegionScaleFactors=(PositiveWidth=0.8,NegativeWidth=0.1,PositiveHeight=0.5,NegativeHeight=0.5)
    ActiveRegionSize=0.2
    Debug=(bDrawDebug=false,bDraw2D=false,bDrawLocalAxes=false,DebugScale=25,DebugSegments=20,DebugThickness=0.25)
    DriverAxis=(X=1,Y=0,Z=0)
    DriverItem=(Type=Bone,Name="head")
    FalloffRegionScaleFactors=(PositiveWidth=0.2,NegativeWidth=0.1,PositiveHeight=0.5,NegativeHeight=0.5)
    FalloffSize=0.45
    OptionalParentItem=(Type=Bone,Name="neck_01")
    RotationOffset=(X=-90,Y=0,Z=50)
   )
   > &RigUnit_ModifyTransforms(
    ItemToModify=((Item=(Type=Bone,Name="dyn_hat_loop_02"),Transform=(Rotation=(X=0,Y=0,Z=0,W=1),Translation=(X=0,Y=-0.8,Z=-0.5),Scale3D=(X=1,Y=1,Z=1))))
    WeightMinimum=0
    ItemToModify.0.Transform.Rotation=&RigVMFunction_MathQuaternionFromEuler(Euler=(X=0,Y=0,Z=2), RotationOrder=XYZ).Result
    Weight=@n1.OutputParam
   )
   > n2 = &RigUnit_SphericalPoseReader(
    ActiveRegionScaleFactors=(PositiveWidth=0.1,NegativeWidth=0.5,PositiveHeight=1.5,NegativeHeight=1)
    ActiveRegionSize=0.2
    Debug=(bDrawDebug=false,bDraw2D=false,bDrawLocalAxes=false,DebugScale=25,DebugSegments=20,DebugThickness=0.25)
    DriverAxis=(X=1,Y=0,Z=0)
    DriverItem=(Type=Bone,Name="head")
    FalloffRegionScaleFactors=(PositiveWidth=0.1,NegativeWidth=0.22,PositiveHeight=0.2,NegativeHeight=0.2)
    FalloffSize=0.45
    OptionalParentItem=(Type=Bone,Name="neck_01")
    RotationOffset=(X=-90,Y=0,Z=120)
   )
   > &RigUnit_ModifyTransforms(
    ItemToModify="((Item=(Type=Bone,Name=\"dyn_hat_loop_02\"),Transform=(Rotation=(X=0,Y=0,Z=0,W=1),Translation=(X=0,Y=0,Z=0),Scale3D=(X=1,Y=1,Z=1))),(Item=(Type=Bone,Name=\"dyn_hat_loop_04\"),Transform=(Rotation=(X=0,Y=0,Z=0,W=1),Translation=(X=0,Y=0,Z=0),Scale3D=(X=1,Y=1,Z=1) ...(truncated)"
    WeightMinimum=0
    ItemToModify.0.Transform.Rotation=&RigVMFunction_MathQuaternionFromEuler(Euler=(X=0,Y=0,Z=10), RotationOrder=XYZ).Result
    ItemToModify.1.Transform.Rotation=&RigVMFunction_MathQuaternionFromEuler(Euler=(X=0,Y=0,Z=0), RotationOrder=XYZ).Result
    Weight=@n2.OutputParam
   )
   > &RigUnit_ParentConstraint(
    AdvancedSettings=(InterpolationType=Average,RotationOrderForFilter=XZY)
    Child=(Type=Bone,Name="dyn_straw_01")
    Parents=((Item=(Type=Bone,Name="C_jaw"),Weight=0.8671),(Item=(Type=Bone,Name="L_lip_lower_outer"),Weight=0.1329))
    Weight=1
   )
   > &RigUnit_ModifyTransforms(
    ItemToModify=((Item=(Type=Bone,Name="dyn_straw_01"),Transform=(Rotation=(X=0,Y=0,Z=0,W=1),Translation=(X=0,Y=0,Z=0.35),Scale3D=(X=1,Y=1,Z=1))))
    Weight=1
    WeightMinimum=0
    ItemToModify.0.Transform.Rotation=&RigVMFunction_MathQuaternionFromEuler(Euler=(X=0,Y=0,Z=0), RotationOrder=XYZ).Result
   )
   > n3 = &RigUnit_SphericalPoseReader(
    ActiveRegionScaleFactors=(PositiveWidth=0.8,NegativeWidth=0.1,PositiveHeight=1,NegativeHeight=1)
    ActiveRegionSize=0.2
    Debug=(bDrawDebug=true,bDraw2D=false,bDrawLocalAxes=false,DebugScale=25,DebugSegments=20,DebugThickness=0.25)
    DriverAxis=(X=1,Y=0,Z=0)
    DriverItem=(Type=Bone,Name="head")
    FalloffRegionScaleFactors=(PositiveWidth=0.2,NegativeWidth=0.1,PositiveHeight=0.5,NegativeHeight=0.5)
    FalloffSize=0.45
    OptionalParentItem=(Type=Bone,Name="neck_01")
    RotationOffset=(X=-90,Y=0,Z=50)
   )
   n4 = RigVMVariableNode() {variable=Melee}
   n5 = &RigVMFunction_AlphaInterp(InterpSpeedDecreasing=10, InterpSpeedIncreasing=10, bClampResult=False, bInterpResult=False, Value=@n3.OutputParam, Scale=@n4.Value)
   > &RigUnit_ModifyTransforms(
    ItemToModify=((Item=(Type=Bone,Name="dyn_hat_loop_02"),Transform=(Rotation=(X=0,Y=0,Z=0,W=1),Translation=(X=4,Y=-1,Z=1),Scale3D=(X=1,Y=1,Z=1))))
    WeightMinimum=0
    Weight=@n5.Result
    ItemToModify.0.Transform.Rotation=&RigVMFunction_MathQuaternionFromEuler(Euler=(X=0,Y=0,Z=0), RotationOrder=XYZ).Result
   )
   > &RigUnit_ParentConstraint(
    AdvancedSettings=(InterpolationType=Average,RotationOrderForFilter=XZY)
    Child=(Type=Bone,Name="dyn_hat_loop_03")
    Filter=(TranslationFilter=(bX=True,bY=True,bZ=True),RotationFilter=(bX=false,bY=false,bZ=false),ScaleFilter=(bX=false,bY=false,bZ=false))
    Parents=((Item=(Type=Bone,Name="neck_02"),Weight=0.13),(Item=(Type=Bone,Name="spine_05"),Weight=0.86))
    Weight=@n4.Value
   )
   > &RigUnit_ParentConstraint(
    AdvancedSettings=(InterpolationType=Average,RotationOrderForFilter=XZY)
    Child=(Type=Bone,Name="dyn_string_r_01")
    Parents=((Item=(Type=Bone,Name="spine_05"),Weight=1))
    Weight=@n4.Value
   )
   > &RigUnit_ParentConstraint(
    AdvancedSettings=(InterpolationType=Average,RotationOrderForFilter=XZY)
    Child=(Type=Bone,Name="dyn_string_l_01")
    Parents=((Item=(Type=Bone,Name="spine_05"),Weight=1))
    Weight=@n4.Value
   )
   AlphaInterp = &RigVMFunction_AlphaInterp(InterpSpeedDecreasing=8, InterpSpeedIncreasing=5, Scale=1, Value=0, bClampResult=False)
  ]}
]
hierarchy [
 root {loc=<0 0 0>, rot=<0 0 0 1>}
 pelvis {parent=root, loc=<0 -0.5301 98.0571>, rot=<0.0127 0.707 -0.0127 -0.707>}
 spine_01 {parent=pelvis, loc=<6.1208 -0.9328 0>, rot=<0 0 0.034 -0.9994>}
 spine_02 {parent=spine_01, loc=<6.5996 -0.2095 0>, rot=<0 0 0.0238 -0.9997>}
 spine_03 {parent=spine_02, loc=<6.4914 -0.1237 0>, rot=<0 0 0.0149 -0.9999>}
 spine_04 {parent=spine_03, loc=<6.9598 0.0794 0>, rot=<0 0 -0.0132 -0.9999>}
 spine_05 {parent=spine_04, loc=<6.9987 0 0>, rot=<0 0 -0.0015 -1>}
 neck_01 {parent=spine_05, loc=<11.3138 0.1529 0>, rot=<0 0 -0.2003 -0.9797>}
 neck_02 {parent=neck_01, loc=<4.5627 0 0>, rot=<0 0 0.0504 -0.9987>}
 head {parent=neck_02, loc=<5.3914 0 0>, rot=<0 0 0.115 -0.9934>}
 dyn_hat_loop_01 {parent=head, loc=<12.7202 -6.2444 -0.0029>, rot=<-0.3655 -0.9308 0.0042 0>}
 dyn_hat_loop_02 {parent=dyn_hat_loop_01, loc=<13.4344 -0.8861 0>, rot=<0 0 -0.0639 -0.998>}
 dyn_hat_loop_03 {parent=dyn_hat_loop_02, loc=<8.13 0.0003 0>, rot=<0 0 0.0943 -0.9955>}
 dyn_hat_loop_04 {parent=dyn_hat_loop_03, loc=<5.5397 -0.0003 0>, rot=<0 0 0.162 -0.9868>}
 dyn_string_l_MSTR {parent=dyn_hat_loop_04, loc=<2.9529 -0.2708 0.4017>, rot=<0.0277 0.0613 0.0024 -0.9977>}
 dyn_string_l_01 {parent=dyn_string_l_MSTR, loc=<0 0 0>, rot=<0 0 0 1>}
 dyn_string_l_02 {parent=dyn_string_l_01, loc=<4.6848 0.0001 0>, rot=<0 0 0 1>}
 dyn_string_l_03 {parent=dyn_string_l_02, loc=<3.8454 0.0001 0>, rot=<0 0 0 1>}
 dyn_string_l_04 {parent=dyn_string_l_03, loc=<3.0302 0.0001 0>, rot=<0 0 0 1>}
 dyn_string_r_MSTR {parent=dyn_hat_loop_04, loc=<3.0459 -0.0467 -0.8133>, rot=<-0.0031 -0.0197 0.0204 -0.9996>}
 dyn_string_r_01 {parent=dyn_string_r_MSTR, loc=<0 0 0>, rot=<0 0 0 1>}
 dyn_string_r_02 {parent=dyn_string_r_01, loc=<8.0728 0 0>, rot=<0 0 0 1>}
 dyn_string_r_03 {parent=dyn_string_r_02, loc=<5.9956 -0.0001 0>, rot=<0 0 0 1>}
 dyn_string_r_04 {parent=dyn_string_r_03, loc=<2.5451 -0.0001 0>, rot=<0 0 0 1>}
 faceAttach {parent=head, loc=<5.1819 -0.7774 0>, rot=<0.0029 -0.7071 -0.0029 -0.7071>}
 C_jaw {parent=faceAttach, loc=<-0.0134 1.5877 -1.4291>, rot=<0 0 0 1>}
 C_lip_lower_mid {parent=C_jaw, loc=<0.0139 23.1533 4.5319>, rot=<0 0 0 -1>}
 L_lip_lower_outer {parent=C_jaw, loc=<3.5317 11.6688 4.3255>, rot=<0.013 0.0445 0.3057 -0.951>}
 dyn_straw_01 {parent=L_lip_lower_outer, loc=<-1.1677 1.2559 0.8766>, rot=<0.7817 0.3627 -0.4507 -0.2331>}
 dyn_straw_02 {parent=dyn_straw_01, loc=<5.0301 -0.2027 0.3886>, rot=<-0.0171 0.0041 -0.0724 -0.9972>}
 dyn_straw_03 {parent=dyn_straw_02, loc=<3.913 0.0425 0.2532>, rot=<0 0 -0.0579 -0.9983>}
 dyn_straw_04 {parent=dyn_straw_03, loc=<3.8614 0.0597 0.354>, rot=<0 0 0 1>}
 dyn_straw_05 {parent=dyn_straw_04, loc=<3.697 0.0573 0.3389>, rot=<-0.0003 -0.0068 -0.0775 -0.997>}
 R_lip_lower_outer {parent=C_jaw, loc=<-3.5049 11.6688 4.3255>, rot=<0.013 -0.0445 -0.3057 -0.951>}
 tongue {parent=C_jaw, loc=<0.0141 6.7661 2.0987>, rot=<-0.1568 0 0 -0.9876>}
 L_lip_corner {parent=faceAttach, loc=<5.2536 6.2899 2.5672>, rot=<0.003 0.008 -0.0146 -0.9999>}
 dyn_hair_01 {parent=head, loc=<10.7695 0.7646 -6.8003>, rot=<0.7069 0.0153 0.707 0.0094>}
 dyn_hair_02 {parent=head, loc=<10.7693 0.7877 6.1383>, rot=<0.707 -0.0094 0.7069 -0.0153>}
]
variables [Melee {type=double, default=0}]
