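# Physics asset for the FlamingoCountry medium female character's face accessory, setting up two dangling string chains hung from the head.
# Left and right chains each link a master bone on the head to four capsule segments, with tight twist and limited swing, narrowest at the first joint.
# Larger capsules on the upper spine and neck, plus a small one on the head, form a free-jointed body chain carrying the strings.
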PhysicsAsset /Game/Characters/Player/Female/Medium/Bodies/F_MED_FlamingoCountry/Meshes/Parts/F_MED_FlamingoCountry_FaceAcc_Physics
bodies {
 dyn_string_l_01 [{shapeType=Capsule, length=4.3899, radius=2.6101, shapeName=dyn_string_l_01_capsule, center=<3 1 0>, rotation=<rot 90 0 0>}]
 dyn_string_l_02 [{shapeType=Capsule, length=5, radius=2, shapeName=dyn_string_l_02_capsule, center=<3 1 0>, rotation=<rot 90 0 0>}]
 dyn_string_l_03 [{shapeType=Capsule, length=5, radius=2, shapeName=dyn_string_l_03_capsule, center=<3 1 0>, rotation=<rot 90 0 0>}]
 dyn_string_l_04 [{shapeType=Capsule, length=4.7219, radius=2.2781, shapeName=dyn_string_l_04_capsule, center=<3 1 0>, rotation=<rot 90 0 0>}]
 dyn_string_l_MSTR [{shapeType=Capsule, length=4, radius=2, shapeName=dyn_string_l_MSTR_capsule, center=<-1.9733 -1.5649 0>}]
 dyn_string_r_01 [{shapeType=Capsule, length=4.3899, radius=2.6101, shapeName=dyn_string_r_01_capsule, center=<2 1 0>, rotation=<rot 90 0 0>}]
 dyn_string_r_02 [{shapeType=Capsule, length=5, radius=2, shapeName=dyn_string_r_02_capsule, center=<2 1 0>, rotation=<rot 90 0 0>}]
 dyn_string_r_03 [{shapeType=Capsule, length=5, radius=2, shapeName=dyn_string_r_03_capsule, center=<2 1 0>, rotation=<rot 90 0 0>}]
 dyn_string_r_04 [{shapeType=Capsule, length=3.7219, radius=3.2781, shapeName=dyn_string_r_04_capsule, center=<2 1 0>, rotation=<rot 90 0 0>}]
 dyn_string_r_MSTR [{shapeType=Capsule, length=4, radius=2, shapeName=dyn_string_r_MSTR_capsule, center=<-1.9733 -1.5649 0>}]
 head [{shapeType=Capsule, length=4, radius=2, shapeName=head_capsule}]
 neck_01 [{shapeType=Capsule, length=10, radius=10.2601, shapeName=neck_01_capsule, rotation=<rot 90 0 0>}]
 neck_02 [{shapeType=Capsule, length=10, radius=10.2601, shapeName=neck_02_capsule, rotation=<rot 90 0 0>}]
 spine_04 [{shapeType=Capsule, length=20.1981, radius=16.2659, shapeName=spine_04_capsule, center=<0.1624 1.7887 0>, rotation=<rot 90 -90 -76.7746>}]
 spine_05 [{shapeType=Capsule, length=19.6513, radius=15.8479, shapeName=spine_05_capsule, center=<1.5547 -1.0784 0>, rotation=<rot 90 0 10.0002>}]
}
constraints [
 {bone1Name=dyn_string_l_01, bone2Name=dyn_string_l_MSTR, swing1LimitDegrees=10, swing2LimitDegrees=10, twistLimitDegrees=2}
 {bone1Name=dyn_string_l_02, bone2Name=dyn_string_l_01, swing1LimitDegrees=20, swing2LimitDegrees=20, twistLimitDegrees=2}
 {bone1Name=dyn_string_l_03, bone2Name=dyn_string_l_02, swing1LimitDegrees=20, swing2LimitDegrees=20, twistLimitDegrees=2}
 {bone1Name=dyn_string_l_04, bone2Name=dyn_string_l_03, swing1LimitDegrees=20, swing2LimitDegrees=20, twistLimitDegrees=2}
 {bone1Name=dyn_string_r_01, bone2Name=dyn_string_r_MSTR, swing1LimitDegrees=10, swing2LimitDegrees=10, twistLimitDegrees=2}
 {bone1Name=dyn_string_r_02, bone2Name=dyn_string_r_01, swing1LimitDegrees=20, swing2LimitDegrees=20, twistLimitDegrees=2}
 {bone1Name=dyn_string_r_03, bone2Name=dyn_string_r_02, swing1LimitDegrees=20, swing2LimitDegrees=20, twistLimitDegrees=2}
 {bone1Name=dyn_string_r_04, bone2Name=dyn_string_r_03, swing1LimitDegrees=20, swing2LimitDegrees=20, twistLimitDegrees=2}
 {
  bone1Name=neck_02
  bone2Name=neck_01
  swing1LimitDegrees=45
  swing1Motion=Free
  swing2LimitDegrees=45
  swing2Motion=Free
  twistLimitDegrees=45
  twistMotion=Free
 }
 {
  bone1Name=neck_01
  bone2Name=spine_05
  swing1LimitDegrees=45
  swing1Motion=Free
  swing2LimitDegrees=45
  swing2Motion=Free
  twistLimitDegrees=45
  twistMotion=Free
 }
 {
  bone1Name=spine_05
  bone2Name=spine_04
  swing1LimitDegrees=45
  swing1Motion=Free
  swing2LimitDegrees=45
  swing2Motion=Free
  twistLimitDegrees=45
  twistMotion=Free
 }
 {
  bone1Name=dyn_string_l_MSTR
  bone2Name=head
  swing1LimitDegrees=45
  swing1Motion=Free
  swing2LimitDegrees=45
  swing2Motion=Free
  twistLimitDegrees=45
  twistMotion=Free
 }
 {
  bone1Name=head
  bone2Name=neck_02
  swing1LimitDegrees=45
  swing1Motion=Free
  swing2LimitDegrees=45
  swing2Motion=Free
  twistLimitDegrees=45
  twistMotion=Free
 }
 {
  bone1Name=dyn_string_r_MSTR
  bone2Name=head
  swing1LimitDegrees=45
  swing1Motion=Free
  swing2LimitDegrees=45
  swing2Motion=Free
  twistLimitDegrees=45
  twistMotion=Free
 }
]
